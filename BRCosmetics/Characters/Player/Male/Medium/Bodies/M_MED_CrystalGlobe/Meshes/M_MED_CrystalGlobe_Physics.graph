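# Physics asset for the M_MED_CrystalGlobe male player cosmetic body, covering its right-side front and leg belt bones.
# Capsule bodies cover base and dyn bones for two front belts and one leg belt, plus the pelvis and right thigh twist.
# Dyn front belts swing mainly on one axis from their bases, while the leg belt is almost rigid.
# Fully free angular constraints link the two front belts, and the pelvis to the thigh and leg belt base.
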PhysicsAsset /BRCosmetics/Characters/Player/Male/Medium/Bodies/M_MED_CrystalGlobe/Meshes/M_MED_CrystalGlobe_Physics
bodies {
 base_Front_Belt_R_01 [{shapeType=Capsule, length=0.505, radius=0.505, shapeName=base_Front_Belt_R_01_capsule, rotation=<rot 0 -90 0>}]
 base_Front_Belt_R_02 [{shapeType=Capsule, length=0.505, radius=0.505, shapeName=base_Front_Belt_R_02_capsule, rotation=<rot 0 -90 0>}]
 base_Leg_Belt_R_01 [{shapeType=Capsule, length=7.8934, radius=3.7896, shapeName=base_Leg_Belt_R_01_capsule, center=<0 -0.1981 -3.0832>, rotation=<rot 0 -90 0>}]
 dyn_Front_Belt_R_01 [{
   shapeType=Capsule
   length=8.0699
   radius=1.8109
   shapeName=dyn_Front_Belt_R_01_capsule
   center=<0.5512 0.8158 -5.2437>
   rotation=<rot 0.2078 -90 3.1759>
  }]
 dyn_Front_Belt_R_02 [{
   shapeType=Capsule
   length=7.9509
   radius=1.8213
   shapeName=dyn_Front_Belt_R_02_capsule
   center=<1.0261 1.5353 -5.1211>
   rotation=<rot 0.4399 -90 -0.2463>
  }]
 dyn_Leg_Belt_R_01 [{shapeType=Capsule, length=9.5409, radius=3.486, center=<0.5991 0 -2.8471>}]
 pelvis [
  {
   shapeType=Capsule
   length=3.5186
   radius=11.8127
   shapeName=pelvis_capsule
   center=<-1.5174 0.0315 0.2421>
   rotation=<rot -8.9023 -90 2.0109>
  }
  {
   shapeType=Capsule
   length=1.6997
   radius=13.9808
   shapeName=pelvis_capsule
   center=<4.9023 17.9215 30.7703>
   rotation=<rot 9.4957 -7.9547 -56.3675>
  }
 ]
 thigh_twist_01_r [{
   shapeType=Capsule
   length=19.4763
   radius=8.0179
   shapeName=thigh_r_capsule
   center=<-9.1148 -0.8647 0.2924>
   rotation=<rot 88.1413 -4.8448 -1.903>
  }]
}
constraints [
 {bone1Name=dyn_Front_Belt_R_01, bone2Name=base_Front_Belt_R_01, swing1LimitDegrees=3, swing2LimitDegrees=50, twistLimitDegrees=0}
 {bone1Name=dyn_Front_Belt_R_02, bone2Name=base_Front_Belt_R_02, swing1LimitDegrees=3, swing2LimitDegrees=60, twistLimitDegrees=0}
 {
  bone1Name=thigh_twist_01_r
  bone2Name=pelvis
  swing1LimitDegrees=45
  swing1Motion=Free
  swing2LimitDegrees=45
  swing2Motion=Free
  twistLimitDegrees=45
  twistMotion=Free
 }
 {
  bone1Name=base_Leg_Belt_R_01
  bone2Name=pelvis
  swing1LimitDegrees=45
  swing1Motion=Free
  swing2LimitDegrees=45
  swing2Motion=Free
  twistLimitDegrees=45
  twistMotion=Free
 }
 {bone1Name=dyn_Leg_Belt_R_01, bone2Name=base_Leg_Belt_R_01, swing1LimitDegrees=0, swing2LimitDegrees=2, twistLimitDegrees=0}
 {
  bone1Name=dyn_Front_Belt_R_01
  bone2Name=dyn_Front_Belt_R_02
  swing1LimitDegrees=45
  swing1Motion=Free
  swing2LimitDegrees=45
  swing2Motion=Free
  twistLimitDegrees=45
  twistMotion=Free
 }
]
