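# Post-pose Anim Blueprint for the CrystalGlobe male medium body cosmetic, adding procedural and physics motion to its belts and wires.
# It copies the attached parent's pose through a Control Rig, then drives belt and arm-wire bones with weighted constraints.
# Spring AnimDynamics sways arm and back wires, followed by additive belt offsets (mostly disabled) and a pelvis-space RigidBody simulation.
# The event graph updates gravity-override and crouch-walking variables.
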AnimBlueprint /BRCosmetics/Characters/Player/Male/Medium/Bodies/M_MED_CrystalGlobe/Meshes/M_MED_CrystalGlobe_AnimBP
target_skeleton=&/M_MED_CrystalGlobe_Skeleton
graphs [
 AnimGraph {nodes [
   n1 = Control Rig(Source=Copy Pose From Mesh() {bUseAttachedParent=true}) {
    alphaBoolBlend {blendOption=Linear}
    controlRigAssetReference {blueprintRigClass=&/M_MED_CrystalGlobe_ControlRig.M_MED_CrystalGlobe_ControlRig_C}
   }
   n2 = "Constraint - base_Chest_Belt_L_01 (PPPPPPP)"!(
    ComponentPose=Local To Component(LocalPose=@n1)
    Alpha=1
    ConstraintWeights_0=0.3496
    ConstraintWeights_1:float=0.2302
    ConstraintWeights_2:float=0.1401
    ConstraintWeights_3:float=0.1024
    ConstraintWeights_4:float=0.0864
    ConstraintWeights_5:float=0.0564
    ConstraintWeights_6:float=0.035
   ) -> (Pose) {
    alphaBoolBlend {blendOption=Linear}
    boneToModify {boneName=base_Chest_Belt_L_01}
    constraintSetup [
     {targetBone {boneName=spine_03}}
     {targetBone {boneName=spine_02}}
     {targetBone {boneName=spine_04}}
     {targetBone {boneName=spine_01}}
     {targetBone {boneName=clavicle_l}}
     {targetBone {boneName=deform_pec_l}}
     {targetBone {boneName=pelvis}}
    ]
    constraintWeights [1,1,1,1,1,1,1]
   }
   n3 = "Constraint - base_Chest_Belt_L_02 (PPPPPPP)"!(
    ComponentPose=@n2
    Alpha=1
    ConstraintWeights_0=0.3267
    ConstraintWeights_1:float=0.2859
    ConstraintWeights_2:float=0.1442
    ConstraintWeights_3:float=0.0946
    ConstraintWeights_4:float=0.0719
    ConstraintWeights_5:float=0.0588
    ConstraintWeights_6:float=0.0179
   ) -> (Pose) {
    alphaBoolBlend {blendOption=Linear}
    boneToModify {boneName=base_Chest_Belt_L_02}
    constraintSetup [
     {targetBone {boneName=spine_03}}
     {targetBone {boneName=spine_02}}
     {targetBone {boneName=spine_01}}
     {targetBone {boneName=spine_04}}
     {targetBone {boneName=pelvis}}
     {targetBone {boneName=clavicle_l}}
     {targetBone {boneName=deform_pec_l}}
    ]
    constraintWeights [1,1,1,1,1,1,1]
   }
   n4 = "Constraint - base_Chest_Belt_R_01 (PPPPPPP)"!(
    ComponentPose=@n3
    Alpha=1
    ConstraintWeights_0=0.3496
    ConstraintWeights_1:float=0.2302
    ConstraintWeights_2:float=0.14
    ConstraintWeights_3:float=0.1024
    ConstraintWeights_4:float=0.0864
    ConstraintWeights_5:float=0.0564
    ConstraintWeights_6:float=0.035
   ) -> (Pose) {
    alphaBoolBlend {blendOption=Linear}
    boneToModify {boneName=base_Chest_Belt_R_01}
    constraintSetup [
     {targetBone {boneName=spine_03}}
     {targetBone {boneName=spine_02}}
     {targetBone {boneName=spine_04}}
     {targetBone {boneName=spine_01}}
     {targetBone {boneName=clavicle_r}}
     {targetBone {boneName=deform_pec_r}}
     {targetBone {boneName=pelvis}}
    ]
    constraintWeights [1,1,1,1,1,1,1]
   }
   n5 = "Constraint - base_Chest_Belt_R_02 (PPPPPPP)"!(
    ComponentPose=@n4
    Alpha=1
    ConstraintWeights_0=0.3267
    ConstraintWeights_1:float=0.2859
    ConstraintWeights_2:float=0.1442
    ConstraintWeights_3:float=0.0946
    ConstraintWeights_4:float=0.0719
    ConstraintWeights_5:float=0.0588
    ConstraintWeights_6:float=0.0179
   ) -> (Pose) {
    alphaBoolBlend {blendOption=Linear}
    boneToModify {boneName=base_Chest_Belt_R_02}
    constraintSetup [
     {targetBone {boneName=spine_03}}
     {targetBone {boneName=spine_02}}
     {targetBone {boneName=spine_01}}
     {targetBone {boneName=spine_04}}
     {targetBone {boneName=pelvis}}
     {targetBone {boneName=clavicle_r}}
     {targetBone {boneName=deform_pec_r}}
    ]
    constraintWeights [1,1,1,1,1,1,1]
   }
   n6 = "Constraint - base_Leg_Belt_R_01 (P)"(ComponentPose=@n5, Alpha=0.75, ConstraintWeights_0=1) {
    alphaBoolBlend {blendOption=Linear}
    boneToModify {boneName=base_Leg_Belt_R_01}
    constraintSetup [{targetBone {boneName=thigh_twist_01_r}}]
    constraintWeights [1]
   }
   n7 = "Constraint - base_arm_wires_l (P)"(ComponentPose=@n6, Alpha=1, ConstraintWeights_0=1) {
    alphaBoolBlend {blendOption=Linear}
    boneToModify {boneName=base_arm_wires_l}
    constraintSetup [{targetBone {boneName=upperarm_twist_02_l}}]
    constraintWeights [1]
   }
   n8 = "Constraint - base_arm_wires_r (P)"(ComponentPose=@n7, Alpha=1, ConstraintWeights_0=1) {
    alphaBoolBlend {blendOption=Linear}
    boneToModify {boneName=base_arm_wires_r}
    constraintSetup [{targetBone {boneName=upperarm_twist_02_r}}]
    constraintWeights [1]
   }
   n9 = "Constraint - base_Side_Belt_L_01 (PPPPPP)"!(
    ComponentPose=@n8
    Alpha=1
    ConstraintWeights_0=0.7683
    ConstraintWeights_1:float=0.1276
    ConstraintWeights_2:float=0.0619
    ConstraintWeights_3:float=0.018
    ConstraintWeights_4:float=0.0153
    ConstraintWeights_5:float=0.0088
   ) -> (Pose) {
    alphaBoolBlend {blendOption=Linear}
    boneToModify {boneName=base_Side_Belt_L_01}
    constraintSetup [
     {targetBone {boneName=pelvis}}
     {targetBone {boneName=thigh_twist_01_l}}
     {targetBone {boneName=deform_groin_l}}
     {targetBone {boneName=spine_02}}
     {targetBone {boneName=spine_01}}
     {targetBone {boneName=spine_03}}
    ]
    constraintWeights [1,1,1,1,1,1]
   }
   n10 = "Constraint - base_Side_Belt_L_02 (PPPPPPP)"!(
    ComponentPose=@n9
    Alpha=1
    ConstraintWeights_0=0.824
    ConstraintWeights_1:float=0.0729
    ConstraintWeights_2:float=0.0299
    ConstraintWeights_3:float=0.0256
    ConstraintWeights_4:float=0.022
    ConstraintWeights_5:float=0.0139
    ConstraintWeights_6:float=0.0118
   ) -> (Pose) {
    alphaBoolBlend {blendOption=Linear}
    boneToModify {boneName=base_Side_Belt_L_02}
    constraintSetup [
     {targetBone {boneName=pelvis}}
     {targetBone {boneName=thigh_twist_01_l}}
     {targetBone {boneName=deform_glute_l}}
     {targetBone {boneName=spine_02}}
     {targetBone {boneName=spine_01}}
     {targetBone {boneName=spine_03}}
     {targetBone {boneName=deform_groin_l}}
    ]
    constraintWeights [1,1,1,1,1,1,1]
   }
   n11 = "Constraint - base_Side_Belt_L_03 (PPPPPP)"!(
    ComponentPose=@n10
    Alpha=1
    ConstraintWeights_0=0.8585
    ConstraintWeights_1:float=0.0378
    ConstraintWeights_2:float=0.0334
    ConstraintWeights_3:float=0.0285
    ConstraintWeights_4:float=0.0252
    ConstraintWeights_5:float=0.0166
   ) -> (Pose) {
    alphaBoolBlend {blendOption=Linear}
    boneToModify {boneName=base_Side_Belt_L_03}
    constraintSetup [
     {targetBone {boneName=pelvis}}
     {targetBone {boneName=deform_glute_l}}
     {targetBone {boneName=thigh_twist_01_l}}
     {targetBone {boneName=spine_02}}
     {targetBone {boneName=spine_01}}
     {targetBone {boneName=spine_03}}
    ]
    constraintWeights [1,1,1,1,1,1]
   }
   n12 = "Constraint - base_Front_Belt_R_01 (PPPPPP)"!(
    ComponentPose=@n11
    Alpha=1
    ConstraintWeights_0=0.5
    ConstraintWeights_1:float=0
    ConstraintWeights_2:float=0
    ConstraintWeights_3:float=0
    ConstraintWeights_4:float=0.25
    ConstraintWeights_5:float=0.25
   ) -> (Pose) {
    alphaBoolBlend {blendOption=Linear}
    boneToModify {boneName=base_Front_Belt_R_01}
    constraintSetup [
     {targetBone {boneName=pelvis}}
     {targetBone {boneName=spine_02}}
     {targetBone {boneName=spine_01}}
     {targetBone {boneName=spine_03}}
     {targetBone {boneName=deform_groin_r}}
     {targetBone {boneName=thigh_twist_01_r}}
    ]
    constraintWeights [1,1,1,1,1,1]
   }
   n13 = "Constraint - base_Front_Belt_R_02 (PPPPPP)"!(
    ComponentPose=@n12
    Alpha=1
    ConstraintWeights_0=0.5
    ConstraintWeights_1:float=0.25
    ConstraintWeights_2:float=0.25
    ConstraintWeights_3:float=0
    ConstraintWeights_4:float=0
    ConstraintWeights_5:float=0
   ) -> (Pose) {
    alphaBoolBlend {blendOption=Linear}
    boneToModify {boneName=base_Front_Belt_R_02}
    constraintSetup [
     {targetBone {boneName=pelvis}}
     {targetBone {boneName=thigh_twist_01_r}}
     {targetBone {boneName=deform_groin_r}}
     {targetBone {boneName=spine_02}}
     {targetBone {boneName=spine_01}}
     {targetBone {boneName=spine_03}}
    ]
    constraintWeights [1,1,1,1,1,1]
   }
   n14 = "Constraint - base_Back_Belt_R_01 (PPPPP)"!(
    ComponentPose=@n13
    Alpha=1
    ConstraintWeights_0=0.8538
    ConstraintWeights_1:float=0.0563
    ConstraintWeights_2:float=0.0495
    ConstraintWeights_3:float=0.0331
    ConstraintWeights_4:float=0.0073
   ) -> (Pose) {
    alphaBoolBlend {blendOption=Linear}
    boneToModify {boneName=base_Back_Belt_R_01}
    constraintSetup [
     {targetBone {boneName=pelvis}}
     {targetBone {boneName=spine_02}}
     {targetBone {boneName=spine_01}}
     {targetBone {boneName=spine_03}}
     {targetBone {boneName=deform_glute_r}}
    ]
    constraintWeights [1,1,1,1,1]
   }
   n15 = "Constraint - base_Back_Belt_R_02 (PPPP)"!(
    ComponentPose=@n14
    Alpha=1
    ConstraintWeights_0=0.839
    ConstraintWeights_1:float=0.0622
    ConstraintWeights_2:float=0.0587
    ConstraintWeights_3:float=0.0401
   ) -> (Pose) {
    alphaBoolBlend {blendOption=Linear}
    boneToModify {boneName=base_Back_Belt_R_02}
    constraintSetup [{targetBone {boneName=pelvis}}, {targetBone {boneName=spine_02}}, {targetBone {boneName=spine_01}}, {targetBone {boneName=spine_03}}]
    constraintWeights [1,1,1,1]
   }
   n16 = Anim Dynamics - Bone: dyn_arm_wires_l!(ComponentPose=@n15, Alpha=1, GravityOverride:Vector="0,0,-980", ExternalForce="0,0,0") -> (Pose) {
    angularDampingOverride=1
    angularSpringConstant=20000
    bAngularSpring=true
    bOverrideAngularDamping=true
    bUseGravityOverride=true
    gravityScale=0.35
    simulationSpace=RootRelative
    alphaBoolBlend {blendOption=Linear}
    boundBone {boneName=dyn_arm_wires_l}
    componentLinearAccScale=<1 1 1>
    componentLinearVelScale=<1 1 1>
    gravityOverride {z=-980}
    physicsBodyDefinitions [{
      boundBone {boneName=dyn_arm_wires_l}
      boxExtents=<20 30 30>
      constraintSetup {angularLimitsMax=<15 15 15>, angularLimitsMin=<-15 -15 -15>, angularTarget=<1 0 0>}
      localJointOffset=<4 0 0>
     }]
   }
   n17 = Anim Dynamics - Bone: dyn_arm_wires_r!(ComponentPose=@n16, Alpha=1, GravityOverride:Vector="0,0,-980", ExternalForce="0,0,0") -> (Pose) {
    angularDampingOverride=1
    angularSpringConstant=20000
    bAngularSpring=true
    bOverrideAngularDamping=true
    bUseGravityOverride=true
    gravityScale=0.35
    simulationSpace=RootRelative
    alphaBoolBlend {blendOption=Linear}
    boundBone {boneName=dyn_arm_wires_r}
    componentLinearAccScale=<1 1 1>
    componentLinearVelScale=<1 1 1>
    gravityOverride {z=-980}
    physicsBodyDefinitions [{
      boundBone {boneName=dyn_arm_wires_r}
      boxExtents=<20 30 30>
      constraintSetup {angularLimitsMax=<15 15 15>, angularLimitsMin=<-15 -15 -15>, angularTarget=<1 0 0>}
      localJointOffset=<4 0 0>
     }]
   }
   n18 = Anim Dynamics - Bone: dyn_back_wires_l!(ComponentPose=@n17, Alpha=1, GravityOverride:Vector="0,0,-980", ExternalForce="0,0,0.3") -> (Pose) {
    angularDampingOverride=1
    angularSpringConstant=20000
    bAngularSpring=true
    bOverrideAngularDamping=true
    bUseGravityOverride=true
    gravityScale=0.35
    simulationSpace=RootRelative
    alphaBoolBlend {blendOption=Linear}
    boundBone {boneName=dyn_back_wires_l}
    componentLinearAccScale=<1 1 1>
    componentLinearVelScale=<1 1 1>
    gravityOverride {z=-980}
    physicsBodyDefinitions [{
      boundBone {boneName=dyn_back_wires_l}
      boxExtents=<20 30 30>
      constraintSetup {angularLimitsMax=<15 15 15>, angularLimitsMin=<-15 -15 -25>, angularTarget=<1 0 0>}
      localJointOffset=<4 0 0>
     }]
   }
   n19 = Anim Dynamics - Bone: dyn_back_wires_r!(ComponentPose=@n18, Alpha=1, GravityOverride:Vector="0,0,-980", ExternalForce="0,0,0.3") -> (Pose) {
    angularDampingOverride=1
    angularSpringConstant=20000
    bAngularSpring=true
    bOverrideAngularDamping=true
    bUseGravityOverride=true
    gravityScale=0.35
    simulationSpace=RootRelative
    alphaBoolBlend {blendOption=Linear}
    boundBone {boneName=dyn_back_wires_r}
    componentLinearAccScale=<1 1 1>
    componentLinearVelScale=<1 1 1>
    gravityOverride {z=-980}
    physicsBodyDefinitions [{
      boundBone {boneName=dyn_back_wires_r}
      boxExtents=<20 30 30>
      constraintSetup {angularLimitsMax=<15 15 25>, angularLimitsMin=<-15 -15 -15>, angularTarget=<1 0 0>}
      localJointOffset=<4 0 0>
     }]
   }
   n20 = "Transform (Modify) Bone - Bone: base_Front_Belt_R_01"!(ComponentPose=@n19, bAlphaBoolEnabled:bool=false, Translation="1.494,-0.924,-2.387", Rotation="77.925,-28.373,-14.576", Scale="0.9,0.9,0.9") -> (Pose) {
    alphaInputType=Bool
    rotationMode=BMM_Additive
    rotationSpace=BCS_BoneSpace
    scaleMode=BMM_Additive
    scaleSpace=BCS_BoneSpace
    translationMode=BMM_Additive
    translationSpace=BCS_BoneSpace
    alphaBoolBlend {blendOption=Linear}
    boneToModify {boneName=base_Front_Belt_R_01}
    rotation=<rot 77.925 -28.373 -14.576>
    scale=<0.9 0.9 0.9>
    translation=<1.494 -0.924 -2.387>
   }
   n21 = "Transform (Modify) Bone - Bone: base_Front_Belt_R_02"!(ComponentPose=@n20, bAlphaBoolEnabled:bool=false, Translation="1.228,-1.396,-1.691", Rotation="-10.404,-8.120,6.111", Scale="0.9,0.9,0.9") -> (Pose) {
    alphaInputType=Bool
    rotationMode=BMM_Additive
    rotationSpace=BCS_BoneSpace
    scaleMode=BMM_Additive
    scaleSpace=BCS_BoneSpace
    translationMode=BMM_Additive
    translationSpace=BCS_BoneSpace
    alphaBoolBlend {blendOption=Linear}
    boneToModify {boneName=base_Front_Belt_R_02}
    rotation=<rot -10.404 -8.12 6.111>
    scale=<0.9 0.9 0.9>
    translation=<1.228 -1.396 -1.691>
   }
   n22 = "Transform (Modify) Bone - Bone: base_Leg_Belt_R_01"!(ComponentPose=@n21, bAlphaBoolEnabled:bool=false, Translation="0,0,0", Rotation="10.492,-1.152,-3.415", Scale="1,1,1") -> (Pose) {
    alphaInputType=Bool
    rotationMode=BMM_Additive
    rotationSpace=BCS_BoneSpace
    alphaBoolBlend {blendOption=Linear}
    boneToModify {boneName=base_Leg_Belt_R_01}
    rotation=<rot 10.492 -1.152 -3.415>
   }
   n23 = "Transform (Modify) Bone - Bone: dyn_Side_Belt_L_01"!(
    ComponentPose=@n22
    bAlphaBoolEnabled:bool=false
    Translation="0.642,-0.455,0.882"
    Rotation="23.430,-24.232,-34.338"
    Scale="0.85,0.85,0.85"
   ) -> (Pose) {
    alphaInputType=Bool
    rotationMode=BMM_Additive
    rotationSpace=BCS_BoneSpace
    scaleMode=BMM_Additive
    scaleSpace=BCS_BoneSpace
    translationMode=BMM_Additive
    translationSpace=BCS_BoneSpace
    alphaBoolBlend {blendOption=Linear}
    boneToModify {boneName=dyn_Side_Belt_L_01}
    rotation=<rot 23.43 -24.232 -34.338>
    scale=<0.85 0.85 0.85>
    translation=<0.642 -0.455 0.882>
   }
   n24 = "Transform (Modify) Bone - Bone: dyn_Side_Belt_L_01"!(ComponentPose=@n23, Alpha:float=1, Translation="0.642,-0.455,0.882", Rotation="-4.980,2.863,15.501", Scale="1,1,1") -> (Pose) {
    rotationMode=BMM_Additive
    rotationSpace=BCS_BoneSpace
    scaleMode=BMM_Additive
    scaleSpace=BCS_BoneSpace
    translationMode=BMM_Additive
    translationSpace=BCS_BoneSpace
    alphaBoolBlend {blendOption=Linear}
    boneToModify {boneName=dyn_Side_Belt_L_01}
    rotation=<rot -4.98 2.863 15.501>
    translation=<0.642 -0.455 0.882>
   }
   n25 = RigidBody!(ComponentPose=@n24, Alpha, OverrideWorldGravity:Vector="0,0,0", ExternalForce) -> (Pose) {
    bClampLinearTranslationLimitToRefPose=true
    bOverrideWorldGravity=true
    overridePhysicsAsset=&/M_MED_CrystalGlobe_Physics
    simulationSpace=BaseBoneSpace
    alphaBoolBlend {blendOption=Linear}
    alphaScaleBiasClamp {bInterpResult=true}
    baseBoneRef {boneName=pelvis}
    componentAppliedLinearAccClamp=<3000 3000 3000>
    componentLinearAccScale=<0.3 0.3 0.3>
    componentLinearVelScale=<0.3 0.3 0.3>
    simSpaceSettings {
     maxAngularAcceleration=1000
     maxAngularVelocity=3
     maxLinearAcceleration=700
     maxLinearVelocity=500
     worldAlpha=0.3
     externalLinearDragV=<0.2 1>
    }
   }
   Output Pose(Result=Component To Local(ComponentPose=@n25)) {layerGroup=None}
  ]}
 EventGraph {nodes [
   Event BlueprintUpdateAnimation()
   > n1 = Gravity Override State Function DEPRECATED(CharacterPartAnimInstance=Self-Reference(), GameTimeInSeconds=GetGameTimeInSeconds(), GravityOverrideParams=Get Gravity Override Params())
   > Set Gravity Override(Gravity Override=@n1)
   n2 = Get bIsCrouching()
   n3 = Get bIsMoving2D()
   > Set is Crouch Walking(is Crouch Walking:double="To Float (Boolean)"(InBool=AND Boolean(A=@n2, B=@n3))) -> (Output_Get:double)
  ]}
]
variables [
 Gravity Override {type=Vector}
 Gravity Override Params {type=GravityOverrideParamsStruct}
 is Crouch Walking {type=number}
]
